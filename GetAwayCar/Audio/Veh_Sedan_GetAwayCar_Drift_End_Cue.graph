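# Sound cue for the drift-end sound of the GetAwayCar sedan vehicle.
# A Random node picks one of three equally weighted drift-end wave variations and is excluded from branch culling.
# A Modulator at the root varies pitch on top of that pick.
# Playback uses reduced volume, the close-detail vehicle movement attenuation and the interior vehicles sound class.
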SoundCue /GetAwayCar/Audio/Veh_Sedan_GetAwayCar_Drift_End_Cue
root=Modulator_0
nodes [
 {id=Modulator_0, children [Random_1], properties {pitchMax=1.1}}
 {id=Random_1, children [WavePlayer_4,WavePlayer_5,WavePlayer_6], properties {bShouldExcludeFromBranchCulling=true, weights [1,1,1]}}
 {id=WavePlayer_4, properties {soundWaveAssetPtr=&/Veh_GetAwayCar_Drift_End_01}}
 {id=WavePlayer_5, properties {soundWaveAssetPtr=&/Veh_GetAwayCar_Drift_End_02}}
 {id=WavePlayer_6, properties {soundWaveAssetPtr=&/Veh_GetAwayCar_Drift_End_03}}
]
properties {
 attenuationSettings=/VehicleAudio/Settings/Vehicle_Movement_Close_Detail_Att
 maxDistance=12000
 soundClassObject=&Fort_Vehicles_Interior
 volumeMultiplier=0.301
}
