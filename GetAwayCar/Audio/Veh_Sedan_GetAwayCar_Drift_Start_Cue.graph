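# Sound cue for the GetAwayCar sedan's drift start, layering a skid-onset one-shot over a sustained drift loop.
# A Mixer blends the looping wave, faded in by an Enveloper volume curve over 1.5 seconds, with one of three random start waves pitched slightly up.
# It uses close-detail vehicle attenuation and the vehicle interior sound class, with a concurrency override limiting simultaneous instances and retriggering.
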SoundCue /GetAwayCar/Audio/Veh_Sedan_GetAwayCar_Drift_Start_Cue
root=Mixer_0
nodes [
 {
  id=Enveloper_0
  children [Modulator_0]
  properties {pitchCurve {editorCurveData {keys [{time=0}]}}, volumeCurve {editorCurveData {keys [<at 0 0>,<at 0.7482 0.1685>,<at 1.5 1>]}}}
 }
 {id=Mixer_0, children [Enveloper_0,Modulator_1], properties {inputVolume [1.5,0.8]}}
 {id=Modulator_0, children [WavePlayer_0]}
 {id=Modulator_1, children [Random_1], properties {pitchMax=1.2, pitchMin=1.1}}
 {id=Random_1, children [WavePlayer_2,WavePlayer_3,WavePlayer_4], properties {weights [1,1,1]}}
 {id=WavePlayer_0, properties {bLooping=true, soundWaveAssetPtr=&/Veh_GetAwayCar_Drift_Loop}}
 {id=WavePlayer_2, properties {soundWaveAssetPtr=&/Veh_GetAwayCar_Drift_Start_01}}
 {id=WavePlayer_3, properties {soundWaveAssetPtr=&/Veh_GetAwayCar_Drift_Start_02}}
 {id=WavePlayer_4, properties {soundWaveAssetPtr=&/Veh_GetAwayCar_Drift_Start_03}}
]
properties {
 attenuationSettings=/VehicleAudio/Settings/Vehicle_Movement_Close_Detail_Att
 bOverrideConcurrency=true
 maxDistance=12000
 soundClassObject=&Fort_Vehicles_Interior
 volumeMultiplier=0.351
 concurrencyOverrides {maxCount=3, retriggerTime=0.5}
}
