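# Sound cue named for the engine-off sound of the GetAwayCar sedan vehicle.
# It has no root and no nodes, so it references no sound waves and produces no audio as saved.
# Its settings remain assigned: shared vehicle movement attenuation, the vehicle interior sound class and the shared engine concurrency set.
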SoundCue /GetAwayCar/Audio/Veh_Sedan_GetAwayCar_Engine_Off_Cue
root=null
nodes []
properties {
 attenuationSettings=/VehicleAudio/Settings/Vehicle_Movement_Att
 maxDistance=12512
 soundClassObject=&Fort_Vehicles_Interior
 volumeMultiplier=1
 concurrencySet [/VehicleAudio/Settings/Vehicle_Engine_Conc]
}
