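# Sound cue for the GetAwayCar sedan's engine-on sound.
# A Random node picks one of three engine-on wave variants with equal weight and is excluded from branch culling.
# It plays through the vehicle interior sound class, with shared vehicle movement attenuation and engine concurrency settings.
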SoundCue /GetAwayCar/Audio/Veh_Sedan_GetAwayCar_Engine_On_Cue
root=Random_0
nodes [
 {id=Random_0, children [WavePlayer_1,WavePlayer_2,WavePlayer_3], properties {bShouldExcludeFromBranchCulling=true, weights [1,1,1]}}
 {id=WavePlayer_1, properties {soundWaveAssetPtr=&/Veh_GetAwayCar_Engine_On_01}}
 {id=WavePlayer_2, properties {soundWaveAssetPtr=&/Veh_GetAwayCar_Engine_On_02}}
 {id=WavePlayer_3, properties {soundWaveAssetPtr=&/Veh_GetAwayCar_Engine_On_03}}
]
properties {
 attenuationSettings=/VehicleAudio/Settings/Vehicle_Movement_Att
 maxDistance=12512
 soundClassObject=&Fort_Vehicles_Interior
 volumeMultiplier=1
 concurrencySet [/VehicleAudio/Settings/Vehicle_Engine_Conc]
}
