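# Sound cue for the GetAway Car sedan's downshift sound, played as an exterior vehicle sound.
# A Random node picks among five equally weighted shift-down wave variations and is excluded from branch culling.
# A Modulator node at the root processes the chosen wave.
# The cue plays at reduced volume through the shared close-detail vehicle movement attenuation.
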SoundCue /GetAwayCar/Audio/Veh_Sedan_GetAwayCar_Shift_Down_Cue
root=Modulator_1
nodes [
 {id=Modulator_1, children [Random_1]}
 {
  id=Random_1
  children [WavePlayer_4,WavePlayer_5,WavePlayer_6,WavePlayer_7,WavePlayer_8]
  properties {bShouldExcludeFromBranchCulling=true, weights [1,1,1,1,1]}
 }
 {id=WavePlayer_4, properties {soundWaveAssetPtr=&/Veh_GetAwayCar_Shift_Down_01}}
 {id=WavePlayer_5, properties {soundWaveAssetPtr=&/Veh_GetAwayCar_Shift_Down_02}}
 {id=WavePlayer_6, properties {soundWaveAssetPtr=&/Veh_GetAwayCar_Shift_Down_03}}
 {id=WavePlayer_7, properties {soundWaveAssetPtr=&/Veh_GetAwayCar_Shift_Down_04}}
 {id=WavePlayer_8, properties {soundWaveAssetPtr=&/Veh_GetAwayCar_Shift_Down_05}}
]
properties {
 attenuationSettings=/VehicleAudio/Settings/Vehicle_Movement_Close_Detail_Att
 maxDistance=12000
 soundClassObject=&Fort_Vehicles_Exterior
 volumeMultiplier=0.4
}
