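# Sound cue for the upshift sound of the GetAwayCar sedan vehicle.
# A Random node picks one of five equally weighted shift-up wave variations and feeds a Modulator that varies pitch and volume on each play.
# It plays at reduced volume in the exterior vehicle sound class, using shared close-detail vehicle movement attenuation.
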SoundCue /GetAwayCar/Audio/Veh_Sedan_GetAwayCar_Shift_Up_Cue
root=Modulator_1
nodes [
 {id=Modulator_1, children [Random_0], properties {pitchMax=1.1, volumeMin=0.9}}
 {
  id=Random_0
  children [WavePlayer_0,WavePlayer_1,WavePlayer_2,WavePlayer_3,WavePlayer_4]
  properties {bShouldExcludeFromBranchCulling=true, weights [1,1,1,1,1]}
 }
 {id=WavePlayer_0, properties {soundWaveAssetPtr=&/Veh_GetAwayCar_Shift_Up_05}}
 {id=WavePlayer_1, properties {soundWaveAssetPtr=&/Veh_GetAwayCar_Shift_Up_04}}
 {id=WavePlayer_2, properties {soundWaveAssetPtr=&/Veh_GetAwayCar_Shift_Up_03}}
 {id=WavePlayer_3, properties {soundWaveAssetPtr=&/Veh_GetAwayCar_Shift_Up_02}}
 {id=WavePlayer_4, properties {soundWaveAssetPtr=&/Veh_GetAwayCar_Shift_Up_01}}
]
properties {
 attenuationSettings=/VehicleAudio/Settings/Vehicle_Movement_Close_Detail_Att
 maxDistance=12000
 soundClassObject=&Fort_Vehicles_Exterior
 volumeMultiplier=0.5
}
