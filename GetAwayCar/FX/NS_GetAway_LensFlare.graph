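# Niagara system for the getaway car's front lens flares, a GPU pair of tinted sprites spawned at mirrored side offsets when the vehicle has a driver.
# Particles take the system's VehicleColor and carry streak size and offset attributes alongside two sprite renderers.
# A scratch module samples occlusion in rings, with Visibility scaling alpha.
# A camera-vector CameraScale drives size, brightness and alpha through curves, and time-of-day selection limits flares to evening and night.
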NiagaraSystem /GetAwayCar/FX/NS_GetAway_LensFlare
dependencies {
 usedDataInterfaces [
  &NiagaraDataInterfaceArrayFloat3
  /Script/Niagara.NiagaraDataInterfaceOcclusion
  &NiagaraDataInterfaceCamera
  &NiagaraDataInterfaceCurve
  &NiagaraDataInterfaceArrayColor
 ]
 usedDynamicInputs [&SelectLinearColorFromArray]
 usedModules [
  &SystemState
  &self:SystemSpawnScript.SystemScriptSource.SystemScriptGraph.NiagaraNodeAssignment_0.SetVariables_#1
 ]
}
emitters {Front_LensFlare {
  simTarget=GPUComputeSim
  emitterUpdateScript [
   &EmitterState(Life Cycle Mode=::NewEnumerator1, Loop Duration=1, UseLoopDelay=false)
   &SpawnBurst_Instantaneous(
    Spawn Count=2
    Spawn Probability=&MakeCustomFloatFromBool(Bool=$User.VehicleHasDriver, True Float=1, False Float=0)
    Use Spawn Probability=true
    UseLoopCountLimit=false
   )
  ]
  particleSpawnScript [
   &PartitionParticles()
   &InitializeParticle(
    Color=$System.VehicleColor
    Position Offset=&SelectVectorFromArray(
     Direct Array Index=$Particles.PartitionParticles.Partition
     Vector Selection Array=DataInterface() {propertyValues {FloatData [<0 60 0>, <0 -62 0>]}}
    )
    Sprite Rotation Mode=::NewEnumerator0
    Sprite Size=<125 75>
    Sprite Size Mode=::NewEnumerator4
    UsePositionOffset=true
    Write Material Random=true
   )
   /SetVariables(
    Particles.Streak_01_Offset:NiagaraPosition=&AddVectorToPosition(Position=$Particles.Position, Vector=<0 0 0>)
    Particles.Streak_01_Size:Vector2f=<350 25>
   ) {isSetParameters=true}
  ]
  particleUpdateScript [
   &ParticleState(Kill Particles When Lifetime Has Elapsed=false, Let Infinitely Lived Particles Die When Emitter Deactivates=true)
   &self:ScratchModule(
    NumberOfRings=5
    OcclusionSize=4.5
    Position:NiagaraPosition=&TransformPosition(Position=$Particles.Position, Source Space=::Local, Destination Space=::World)
    SamplesPerRing=5
   ) {module=ScratchModule}
   /SetVariables(Particles.CameraScale:NiagaraFloat=&MakeFloatFromVector(
     Vector=&Multiply_VectorByFloat(
      Vector=&ReturnCameraVector(Camera Vectors=::NewEnumerator4, Coordinate Space=::Simulation, Origin Position=$Particles.Position)
      Float=&DotProduct(
       Vector A=<1 0 -1.5>
       Vector B=&ReturnCameraVector(Camera Vectors=::NewEnumerator4, Coordinate Space=::Simulation, Origin Position=$Particles.Position)
      )
     )
     Channel=::NewEnumerator0
    )) {isSetParameters=true}
   &ScaleSpriteSize(
    Uniform Curve Index=$Particles.CameraScale
    Uniform Curve Sprite Scale=DataInterface() {propertyValues {Curve {keys [
        {arriveTangent=4.7849, interpMode=RCIM_Cubic, leaveTangent=4.7849, tangentMode=RCTM_User, time=0, value=0.5}
        {arriveTangent=0.9156, interpMode=RCIM_Cubic, leaveTangent=0.9156, tangentMode=RCTM_User, time=0.3, value=1.275}
        {tangentMode=RCTM_User, time=1, value=1.5}
       ]}}}
   )
   &ScaleColor(Scale Alpha=$Particles.Visibility, Scale RGB=&VectorFromFloat(Value=3), ScaleA=true, ScaleRGB=true) {module=ScaleColor002}
   &ScaleColor(
    Scale Alpha=&FloatFromCurve(
     FloatCurve=DataInterface() {propertyValues {Curve {keys [
         {interpMode=RCIM_Cubic, tangentMode=RCTM_User, time=0, value=0}
         {arriveTangent=0.4636, interpMode=RCIM_Cubic, leaveTangent=0.4636, tangentMode=RCTM_User, time=0.2, value=0.8}
         {arriveTangentWeight=1.2316, interpMode=RCIM_Cubic, tangentMode=RCTM_User, time=1, value=1}
        ]}}}
     CurveIndex=$Particles.CameraScale
    )
    Scale RGB=&Multiply_VectorByFloat(Vector=<1 1 1>, Float=$Particles.CameraScale)
    ScaleA=true
    ScaleRGB=true
   )
   &ScaleColor(
    Scale Alpha=&FN_Select_Float_from_Time_Of_Day(Morning=0, Day=0, Evening=1, Night=1)
    ScaleA=true
    ScaleRGB=false
   ) {module=ScaleColor001}
  ]
  rendererClasses [&NiagaraSpriteRendererProperties]
  renderers [
   {rendererClass=&NiagaraSpriteRendererProperties}
   {rendererClass=&NiagaraSpriteRendererProperties, rendererIndex=1}
  ]
 }}
summary {
 systemName=NS_GetAway_LensFlare
 emitters [{emitterName=Front_LensFlare, simTarget=GPUComputeSim, rendererClasses [&NiagaraSpriteRendererProperties]}]
 userVariables [
  User.IsDrifting {type=NiagaraBool, defaultValue=true}
  User.SedanColor {type=NiagaraInt32}
  User.VehicleHasDriver {type=NiagaraBool, defaultValue=true}
 ]
}
