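# Sound cue for the GetAwayCar sedan's drift boost blips, playing short flare one-shots on the vehicle interior sound class.
# A Random node picks one of six equally weighted Drift Flare waves and is excluded from branch culling.
# A Modulator then randomizes volume and raises pitch by a random amount for variation.
# Close-detail vehicle movement attenuation applies, and a concurrency override caps simultaneous instances with a retrigger time.
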SoundCue /GetAwayCar/Audio/Veh_Sedan_GetAwayCar_Drift_Boost_Blips_Cue
root=Modulator_1
nodes [
 {id=Modulator_1, children [Random_1], properties {pitchMax=1.5, pitchMin=1, volumeMax=1, volumeMin=0.2}}
 {
  id=Random_1
  children [WavePlayer_2,WavePlayer_3,WavePlayer_4,WavePlayer_7,WavePlayer_8,WavePlayer_9]
  properties {bShouldExcludeFromBranchCulling=true, weights [1,1,1,1,1,1]}
 }
 {id=WavePlayer_2, properties {soundWaveAssetPtr=&/Veh_GetAwayCar_Drift_Flare_01}}
 {id=WavePlayer_3, properties {soundWaveAssetPtr=&/Veh_GetAwayCar_Drift_Flare_02}}
 {id=WavePlayer_4, properties {soundWaveAssetPtr=&/Veh_GetAwayCar_Drift_Flare_03}}
 {id=WavePlayer_7, properties {soundWaveAssetPtr=&/Veh_GetAwayCar_Drift_Flare_04}}
 {id=WavePlayer_8, properties {soundWaveAssetPtr=&/Veh_GetAwayCar_Drift_Flare_05}}
 {id=WavePlayer_9, properties {soundWaveAssetPtr=&/Veh_GetAwayCar_Drift_Flare_06}}
]
properties {
 attenuationSettings=/VehicleAudio/Settings/Vehicle_Movement_Close_Detail_Att
 bOverrideConcurrency=true
 maxDistance=12000
 soundClassObject=&Fort_Vehicles_Interior
 volumeMultiplier=0.4
 concurrencyOverrides {maxCount=3, retriggerTime=0.5}
}
