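# Niagara vehicle FX system for the GetAway car, covering tire dust, peel-out smoke, blown-tire sparks, snow chunks and exhaust.
# User variables feed in vehicle state such as drifting, shifting, speed, driver and fuel.
# ExhaustFlame spawns socket-attached meshes and lights on shifts or speed checks, with drift changing color and size; AudioBlips mirrors that spawn logic for boost sounds.
# NeonLensFlare shows occlusion-faded GPU flares at night while drifting; the neon ribbon trail is disabled.
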NiagaraSystem /GetAwayCar/FX/NS_Wheel_And_Tire_Interaction_GetAway
dependencies=null
emitters {
 AudioBlips {
  emitterSpawnScript [/SetVariables(Emitter.Audio Player=DataInterface() {propertyValues {
      Attenuation=null
      Concurrency=null
      MaxPlaysPerTick=10
      SoundToPlay=/GetAwayCar/Audio/Veh_Sedan_GetAwayCar_Drift_Boost_Blips_Cue
      bAllowLoopingOneShotSounds=false
      bLimitPlaysPerTick=true
      bOnlyActiveDuringGameplay=false
      bStopWhenComponentIsDestroyed=true
      ConfigurationUserParameter {parameter {name=None, typeDefHandle {registeredTypeIndex=1}, varData []}}
      ParameterNames []
     }}) {isSetParameters=true}]
  emitterUpdateScript [
   &EmitterState(Life Cycle Mode=::NewEnumerator1, Loop Duration=0.1, UseLoopDelay=false)
   &SpawnBurst_Instantaneous(
    Spawn Count=4
    Spawn Probability=&MakeCustomFloatFromBool(
     Bool=&Boolean_AndOperation(Bool1=$User.VehicleHasDriver, Bool2=&Boolean_AndOperation(Bool1=$User.VehicleHasFuel, Bool2=$User.IsShifting))
     True Float=1
     False Float=0
    )
    Use Spawn Probability=true
   )
   &SpawnRate(
    Spawn Probability=&MakeCustomFloatFromBool(
     Bool=&Boolean_AndOperation(Bool1=$User.VehicleHasDriver, Bool2=$User.VehicleHasFuel)
     True Float=&MakeCustomFloatFromBool(
      Bool=&Boolean_AndOperation(
       Bool1=&SetBoolByFloatComparison(A=$System.Timeline.NormalizedTime, B=1, Comparison Type=::NewEnumerator2)
       Bool2=&SetBoolByFloatComparison(A=$User.Vehicle_Speed, B=0.2)
      )
      True Float=1
      False Float=0
     )
     False Float=0
    )
    SpawnRate=14
    Use Spawn Probability=true
   )
  ]
  particleSpawnScript [
   &FN_SetSpawnGroup()
   /Niagara/Modules/Audio/PlayPersistentAudio(Audio Player:NiagaraDataInterfaceAudioPlayer=$Emitter.Audio Player)
  ]
  particleUpdateScript [
   &SkeletalMeshLocation(
    Bone / Socket Index=&SelectIntFromArray(Direct Array Index=$Particles.SpawnGroup, Int Selection Array=DataInterface() {propertyValues {IntData [12,13,12,13]}})
    MaskSpawnGroup=false
    Mesh Sampling Type=::NewEnumerator5
    Orientation Sampling=::NewEnumerator1
    Sampled Position Offset=&SelectVectorFromArray(
     Direct Array Index=$Particles.SpawnGroup
     Vector Selection Array=DataInterface() {propertyValues {FloatData [<-8 0 0>, <-8 0 0>, <12 0 0>, <12 0 0>]}}
    )
    Skeletal Mesh=$User.SK_Vehicle
    Socket Sampling Mode=::NewEnumerator5
   )
   /Niagara/Modules/Audio/UpdatePersistentAudio(Audio Player:NiagaraDataInterfaceAudioPlayer=$Emitter.Audio Player, Update Type:ENiagara_AudioParamType=::NewEnumerator5)
   &UpdatePersistentAudio(Audio Player:NiagaraDataInterfaceAudioPlayer=$Emitter.Audio Player, Update Type:ENiagara_AudioParamType=::NewEnumerator6) {module=UpdatePersistentAudio001}
  ]
  rendererClasses []
  renderers []
 }
 ExhaustFlame {
  emitterUpdateScript [
   &EmitterState(Life Cycle Mode=::NewEnumerator1, Loop Duration=0.1, UseLoopDelay=false)
   &SpawnBurst_Instantaneous(
    Spawn Count=4
    Spawn Probability=&MakeCustomFloatFromBool(
     Bool=&Boolean_AndOperation(Bool1=$User.VehicleHasDriver, Bool2=&Boolean_AndOperation(Bool1=$User.VehicleHasFuel, Bool2=$User.IsShifting))
     True Float=1
     False Float=0
    )
    Use Spawn Probability=true
   )
   &SpawnRate(
    Spawn Probability=&MakeCustomFloatFromBool(
     Bool=&Boolean_AndOperation(Bool1=$User.VehicleHasDriver, Bool2=$User.VehicleHasFuel)
     True Float=&MakeCustomFloatFromBool(
      Bool=&Boolean_AndOperation(
       Bool1=&SetBoolByFloatComparison(A=$System.Timeline.NormalizedTime, B=1, Comparison Type=::NewEnumerator2)
       Bool2=&SetBoolByFloatComparison(A=$User.Vehicle_Speed, B=0.2)
      )
      True Float=1
      False Float=0
     )
     False Float=0
    )
    SpawnRate=14
    Use Spawn Probability=true
   )
  ]
  particleSpawnScript [
   &FN_SetSpawnGroup()
   &InitializeParticle(
    Color=&LerpLinearColors(
     StartColor=<rgba 1 0.6269 0.1303 1>
     EndColor=<rgba 1 1 1 1>
     LerpFactor=&MakeCustomFloatFromBool(Bool=$User.IsDrifting, True Float=1, False Float=0)
    )
    Lifetime Max=0.2
    Lifetime Min=0.1
    Lifetime Mode=::NewEnumerator1
    Mesh Scale=&Lerp_Vector(
     A=<0.5 0.6 0.6>
     B=<1 0.75 0.75>
     Alpha=&FloatFromCurve(
      FloatCurve=DataInterface() {propertyValues {Curve {keys [
          {interpMode=RCIM_Cubic, tangentMode=RCTM_Break, time=0, value=0}
          {arriveTangent=1.393, interpMode=RCIM_Cubic, tangentMode=RCTM_Break, time=1, value=1}
         ]}}}
      CurveIndex=&RandomRangeFloat(Minimum=0, Maximum=1)
     )
    )
    Mesh Scale Mode=::NewEnumerator4
    UsePositionOffset=false
    Write Material Random=true
   )
   &SkeletalMeshLocation(
    Bone / Socket Index=&SelectIntFromArray(Direct Array Index=$Particles.SpawnGroup, Int Selection Array=DataInterface() {propertyValues {IntData [12,13,12,13]}})
    Interpolate Mesh Sampling=false
    MaskSpawnGroup=false
    Mesh Sampling Type=::NewEnumerator5
    Orientation Sampling=::NewEnumerator1
    Sampled Position Offset=&SelectVectorFromArray(
     Direct Array Index=$Particles.SpawnGroup
     Vector Selection Array=DataInterface() {propertyValues {FloatData [<-8 0 0>, <-8 0 0>, <12 0 0>, <12 0 0>]}}
    )
    Skeletal Mesh=$User.SK_Vehicle
    Socket Sampling Mode=::NewEnumerator5
   )
   &InitialMeshOrientation(
    Mesh Orientation Mode=::NewEnumerator2
    Rotation=&SelectVectorFromArray(
     Direct Array Index=$Particles.SpawnGroup
     Vector Selection Array=DataInterface() {propertyValues {FloatData [<-1 -1 -1>, <1 -1 1>, <-1 -1 -1>, <1 -1 1>]}}
    )
    Use Rotation Vector=true
   )
   /SetVariables(Particles.LightOffset:Vector3f=&AddVectorToPosition(
     Position=$Particles.SkeletalMeshLocation.SampledPosition
     Vector=&SelectVectorFromArray(Direct Array Index=$Particles.SpawnGroup, Vector Selection Array=DataInterface() {propertyValues {FloatData [<-25 0 0>, <-25 0 0>]}})
    )) {isSetParameters=true}
   /Niagara/Modules/Audio/PlayAudio(Audio Player=DataInterface() {propertyValues {
      Attenuation=null
      Concurrency=null
      MaxPlaysPerTick=10
      SoundToPlay=&Veh_Sedan_GetAwayCar_Drift_Boost_Blips_Cue
      bAllowLoopingOneShotSounds=false
      bLimitPlaysPerTick=true
      bOnlyActiveDuringGameplay=false
      bStopWhenComponentIsDestroyed=true
      ConfigurationUserParameter {parameter {name=None, typeDefHandle {registeredTypeIndex=1}, varData []}}
      ParameterNames []
     }}) {enabled=false}
  ]
  particleUpdateScript [
   &ParticleState()
   &DynamicMaterialParameters(
    Index 1 Param 1=&MakeCustomFloatFromBool(Bool=$User.IsDrifting, True Float=1, False Float=0)
    Module.Index1Param0Write=true
    Module.Index1Param1Write=true
    Module.Index1Param2Write=true
    Module.Index1Param3Write=true
    Write Parameter Index 0=false
    Write Parameter Index 1=true
   )
   &ScaleMeshSize(Scale Factor=&VectorFromCurve(
     VectorCurve=DataInterface() {propertyValues {
       bOptimizeLUT=false
       XCurve {keys [{arriveTangent=-0.4338, interpMode=RCIM_Cubic, leaveTangent=-0.4338, tangentMode=RCTM_User, time=0, value=0.75}, <at 1 2>]}
       YCurve {keys [{arriveTangent=0.1422, interpMode=RCIM_Cubic, leaveTangent=0.1422, tangentMode=RCTM_User, time=0, value=0.75}, <at 1 2>]}
       ZCurve {keys [
         <at -0.0011 1.0067 RCIM_Cubic>
         {arriveTangent=-0.105, interpMode=RCIM_Cubic, leaveTangent=-0.105, tangentMode=RCTM_User, time=1.0016, value=1}
        ]}
      }}
     Scale Curve=&VectorFromFloat(Value=&Lerp_Float(
       A=1
       B=1.4
       Alpha=&FloatFromCurve(
        FloatCurve=DataInterface() {propertyValues {Curve {keys [
            {interpMode=RCIM_Cubic, tangentMode=RCTM_User, time=0, value=1}
            {arriveTangentWeight=1.1078, interpMode=RCIM_Cubic, tangentMode=RCTM_User, time=1, value=0}
           ]}}}
        CurveIndex=$Emitter.NormalizedLoopAge
       )
      ))
    ))
   &ScaleMeshSize(Scale Factor=&Lerp_Vector(
     A=<0.85 0.5 0.5>
     B=<1 1 1>
     Alpha=&MakeCustomFloatFromBool(Bool=$User.IsDrifting, True Float=1, False Float=0)
    )) {module=ScaleMeshSize001}
   &ScaleColor(
    Scale Alpha=&FloatFromCurve(
     FloatCurve=DataInterface() {propertyValues {Curve {keys [
         {interpMode=RCIM_Cubic, leaveTangent=9.8521, tangentMode=RCTM_Break, time=0, value=0.1}
         {interpMode=RCIM_Cubic, tangentMode=RCTM_User, time=0.2, value=1}
         {interpMode=RCIM_Cubic, tangentMode=RCTM_Break, time=1, value=0}
        ]}}}
     Scale Curve=1.2
    )
    Scale RGB=&VectorFromFloat(Value=&Multiply_Float(
      A=&Multiply_Float(
       A=&RandomRangeFloat(Minimum=3, Maximum=25)
       B=&MakeCustomFloatFromBool(Bool=$User.IsDrifting, True Float=1, False Float=0.25)
      )
      B=&FloatFromCurve(
       FloatCurve=DataInterface() {propertyValues {Curve {keys [
           {interpMode=RCIM_Cubic, tangentMode=RCTM_User, time=0.3, value=1}
           {arriveTangentWeight=0.6424, interpMode=RCIM_Cubic, tangentMode=RCTM_User, time=1, value=0.25}
          ]}}}
       CurveIndex=$Emitter.NormalizedLoopAge
      )
     ))
    ScaleA=true
    ScaleRGB=true
   )
   /Game/Effects/Niagara/Modules/Materials/MaterialPanner(
    Jitter Pan by SpawnGroups=false
    Offset:NiagaraFloat=&Add_Float(A=$Emitter.Age, B=$Particles.Initial.MaterialRandom)
    Speed=-5
   )
   &FN_Light_Attributes(
    Color=&LerpLinearColors(
     StartColor=<rgba 1 0.6269 0.1303 1>
     EndColor=<rgba 0 0.19 1 1>
     LerpFactor=&MakeCustomFloatFromBool(Bool=$User.IsDrifting, True Float=1, False Float=0)
    )
    Create Color Attribute=true
    Create Exponent Attribute=true
    Create Position Attribute=true
    Create Radius Attribute=true
    Position=&ConvertPositionToVector(Input Position=$Particles.Position, Convert To Absolute Large World Vector=false)
    Radius=100
    Scale Color=&Multiply_Float(
     A=&Multiply_Float(A=2, B=&Waveform("Waveform [1]"=::NewEnumerator0, "[1] Frequency"=2, "[1] Amplitude Min/Max"=<0.95 1>))
     B=&Lerp_Float(
      A=0.25
      B=1
      Alpha=&FloatFromCurve(
       FloatCurve=DataInterface() {propertyValues {Curve {keys [
           {interpMode=RCIM_Cubic, tangentMode=RCTM_User, time=0, value=1}
           {arriveTangentWeight=1.1078, interpMode=RCIM_Cubic, tangentMode=RCTM_User, time=1, value=0}
          ]}}}
       CurveIndex=$Emitter.NormalizedLoopAge
       Scale Curve=2
      )
     )
    )
   )
  ]
  rendererClasses [&NiagaraMeshRendererProperties, &NiagaraLightRendererProperties]
  renderers [
   {rendererClass=&NiagaraMeshRendererProperties, rendererIndex=0}
   {rendererClass=&NiagaraLightRendererProperties, rendererIndex=1}
  ]
 }
 NeonAnimTrail {
  bEnabled=false
  emitterUpdateScript [
   &EmitterState(Life Cycle Mode=::NewEnumerator1, Loop Behavior=::NewEnumerator1, Loop Duration Mode=::NewEnumerator1, UseLoopDelay=false)
   &FN_SpawnAnimTrail(Enable:NiagaraBool=$User.VehicleHasDriver, Number of Anim Trails=2)
  ]
  particleSpawnScript [
   &FN_InitializeAnimTrailBoneIndecies(First Bone or Socket Index=3, Second Bone or Socket Index=2)
   &FN_InitializeAnimTrailBoneIndecies(AnimTrail Selection=1, First Bone or Socket Index=5, Second Bone or Socket Index=4) {module=FN_InitializeAnimTrailBoneIndecies001}
   &FN_InitializeAnimTrail(
    AnimTrailWidth=1.2
    Lifetime=&Lerp_Float(
     A=0.025
     B=&Lerp_Float(
      A=0.085
      B=0.1
      Alpha=&FNTimeOfDayScalar(Night Scale=1, Morning Scale=0, Day Scale=0, Evening Scale=0)
     )
     Alpha=&FloatFromCurve(
      FloatCurve=DataInterface() {propertyValues {Curve {keys [<at 0.85 0>, {arriveTangent=51.2468, interpMode=RCIM_Cubic, leaveTangent=51.2468, tangentMode=RCTM_User, time=1, value=1}]}}}
      CurveIndex=$User.Vehicle_Speed
     )
    )
   )
   &FN_InitializeAnimTrail(
    AnimTrailWidth=1.2
    Lifetime=&Lerp_Float(
     A=0.05
     B=&Lerp_Float(
      A=0.1
      B=0.23
      Alpha=&FNTimeOfDayScalar(Night Scale=1, Morning Scale=0, Day Scale=0, Evening Scale=0)
     )
     Alpha=1
    )
   ) {enabled=false, module=FN_InitializeAnimTrail001}
   &AddVelocity(Velocity=<1000 0 0>) {enabled=false}
  ]
  particleUpdateScript [
   &ParticleState()
   &FN_UpdateAnimTrail(SkeletalMesh=$User.SkeletalMesh)
   &ScaleColor(
    Scale Alpha=&FloatFromCurve(
     FloatCurve=DataInterface() {propertyValues {Curve {keys [
         {interpMode=RCIM_Cubic, tangentMode=RCTM_User, time=0, value=1}
         {arriveTangent=-2.1393, interpMode=RCIM_Cubic, leaveTangent=-2.1393, tangentMode=RCTM_User, time=0.4, value=0.75}
         {interpMode=RCIM_Cubic, tangentMode=RCTM_User, time=1, value=0}
        ]}}}
     Scale Curve=&FNTimeOfDayScalar(Night Scale=0.75, Morning Scale=0.5, Day Scale=0.5, Evening Scale=0.5)
    )
    Scale RGB=&MakeVectorFromLinearColorRGB(Color=$System.VehicleColor)
    ScaleA=true
    ScaleRGB=true
   )
   &ScaleColor(
    Scale Alpha=&Lerp_Float(
     A=0
     B=1
     Alpha=&FloatFromCurve(
      FloatCurve=DataInterface() {propertyValues {Curve {keys [
          {interpMode=RCIM_Cubic, tangentMode=RCTM_User, time=0.1, value=0}
          {arriveTangent=3.6846, interpMode=RCIM_Cubic, leaveTangent=3.6846, tangentMode=RCTM_User, time=0.9, value=1}
         ]}}}
      CurveIndex=$User.Vehicle_Speed
     )
    )
    Scale RGB=&VectorFromFloat(Value=&Lerp_Float(
      A=0.4
      B=1.25
      Alpha=&FloatFromCurve(
       FloatCurve=DataInterface() {propertyValues {Curve {keys [<at 0.97 0>, {arriveTangent=51.2468, interpMode=RCIM_Cubic, leaveTangent=51.2468, tangentMode=RCTM_User, time=1, value=1}]}}}
       CurveIndex=$User.Vehicle_Speed
      )
     ))
    ScaleA=true
    ScaleRGB=true
   ) {module=ScaleColor001}
   &ScaleColor(
    Scale Alpha=&Lerp_Float(
     A=0
     B=1
     Alpha=&FloatFromCurve(
      FloatCurve=DataInterface() {propertyValues {Curve {keys [{tangentMode=RCTM_User, time=0, value=1}, {tangentMode=RCTM_User, time=1, value=1}]}}}
      CurveIndex=$User.Vehicle_Speed
     )
    )
    Scale RGB=&VectorFromFloat(Value=&Lerp_Float(
      A=0.4
      B=1.25
      Alpha=&FloatFromCurve(
       FloatCurve=DataInterface() {propertyValues {Curve {keys [<at 0.97 1>, {arriveTangent=51.2468, interpMode=RCIM_Cubic, leaveTangent=51.2468, tangentMode=RCTM_User, time=1, value=1}]}}}
       CurveIndex=$User.Vehicle_Speed
      )
     ))
    ScaleA=true
    ScaleRGB=true
   ) {enabled=false, module=ScaleColor002}
   &DynamicMaterialParameters(
    Index 0 Param 1=&FloatFromCurve(FloatCurve=DataInterface() {propertyValues {Curve {keys [
         {interpMode=RCIM_Cubic, leaveTangentWeight=1.4165, tangentMode=RCTM_User, time=0, value=0}
         {arriveTangent=0.1546, interpMode=RCIM_Cubic, leaveTangent=0.1546, tangentMode=RCTM_User, time=0.6, value=1}
         {interpMode=RCIM_Cubic, tangentMode=RCTM_User, time=1, value=1}
        ]}}})
    Index 0 Param 2=5
    Index 0 Param 3=0.1
    Module.Param0WriteEnabled=true
    Module.Param1WriteEnabled=true
    Module.Param2WriteEnabled=true
    Module.Param3WriteEnabled=true
   )
   &SolveForcesAndVelocity(Clamp Velocity=false, Limit Acceleration=false) {enabled=false}
  ]
  rendererClasses [&NiagaraRibbonRendererProperties]
  renderers [{rendererClass=&NiagaraRibbonRendererProperties}]
 }
 NeonLensFlare {
  simTarget=GPUComputeSim
  emitterUpdateScript [
   &EmitterState(Life Cycle Mode=::NewEnumerator1, Loop Behavior=::NewEnumerator1, Loop Duration Mode=::NewEnumerator1, UseLoopDelay=false)
   &SpawnBurst_Instantaneous(Spawn Count=2, Use Spawn Probability=true, UseLoopCountLimit=false)
  ]
  particleSpawnScript [
   &PartitionParticles()
   &InitializeParticle(Color=<rgba 1 1 1 0.75>, Sprite Size=<300 100>, Sprite Size Mode=::NewEnumerator4, UsePositionOffset=false, Write Material Random=true)
   &SkeletalMeshLocation(
    Bone / Socket Index=&SelectIntFromArray(Direct Array Index=$Particles.PartitionParticles.Partition, Int Selection Array=DataInterface() {propertyValues {IntData [0,1]}})
    Interpolate Mesh Sampling=false
    MaskSpawnGroup=false
    Mesh Sampling Type=::NewEnumerator5
    Sampled Position Offset=&SelectVectorFromArray(
     Direct Array Index=$Particles.PartitionParticles.Partition
     Vector Selection Array=DataInterface() {propertyValues {FloatData [<0 -12.5 0>, <0 12.5 0>]}}
    )
    Skeletal Mesh=$User.SkeletalMesh
    Socket Sampling Mode=::NewEnumerator5
   )
   &SpriteFacingAndAlignment(Sprite Alignment=<0 0 1>, WriteAlignment=true, WriteFacing=true)
   &CameraOffset(Camera Offset Amount=75)
   /SetVariables(
    Particles.ParticleSize_alt:Vector2f=<300 80>
    Particles.Position_Alt:NiagaraPosition=&AddVectorToPosition(
     Position=$Particles.SkeletalMeshLocation.SampledPosition
     Vector=&SelectVectorFromArray(
      Direct Array Index=$Particles.PartitionParticles.Partition
      Vector Selection Array=DataInterface() {propertyValues {FloatData [<1 10 0>, <1 -10 0>]}}
     )
    )
   ) {isSetParameters=true}
  ]
  particleUpdateScript [
   &ParticleState(Kill Particles When Lifetime Has Elapsed=false, Let Infinitely Lived Particles Die When Emitter Deactivates=true)
   &self:ScratchModule_01(OcclusionSize=2) {module=ScratchModule_01}
   /SetVariables(Particles.CameraScale:NiagaraFloat=&Multiply_Float(
     A=2
     B=&MakeFloatFromVector(
      Vector=&ReturnCameraVector(Camera Vectors=::NewEnumerator4, Coordinate Space=::Simulation, Origin Position=$Particles.Position)
      Channel=::NewEnumerator0
     )
    )) {isSetParameters=true}
   &ScaleSpriteSize(
    Scale Factor=&MakeVector2D(
     X=&Lerp_Float(
      A=0.1
      B=0.7
      Alpha=&FloatFromCurve(
       FloatCurve=DataInterface() {propertyValues {Curve {keys [
           {interpMode=RCIM_Cubic, tangentMode=RCTM_User, time=0, value=0}
           {arriveTangent=1.6066, interpMode=RCIM_Cubic, leaveTangent=1.6066, tangentMode=RCTM_User, time=2, value=1}
          ]}}}
       CurveIndex=$Particles.CameraScale
      )
     )
     Y=&Lerp_Float(
      A=0.25
      B=1
      Alpha=&FloatFromCurve(
       FloatCurve=DataInterface() {propertyValues {Curve {keys [
           {interpMode=RCIM_Cubic, tangentMode=RCTM_User, time=0, value=0}
           {arriveTangent=3.4168, interpMode=RCIM_Cubic, leaveTangent=3.4168, tangentMode=RCTM_User, time=1, value=1}
          ]}}}
       CurveIndex=$Particles.CameraScale
      )
     )
    )
    Scale Sprite Size Mode=::NewEnumerator2
   )
   &ScaleSpriteSize(
    Initial Sprite Size:Vector2f=$Particles.Initial.ParticleSize_alt
    Scale Factor=&MakeVector2D(
     X=1
     Y=&Lerp_Float(
      A=0.1
      B=1
      Alpha=&FloatFromCurve(
       FloatCurve=DataInterface() {propertyValues {Curve {keys [
           {interpMode=RCIM_Cubic, tangentMode=RCTM_User, time=0, value=0}
           {arriveTangent=3.6046, interpMode=RCIM_Cubic, leaveTangent=3.6046, tangentMode=RCTM_User, time=1, value=1}
          ]}}}
       CurveIndex=$Particles.CameraScale
      )
     )
    )
    Scale Sprite Size Mode=::NewEnumerator2
   ) {module=ScaleSpriteSize001}
   &ScaleColor(
    Scale Alpha=&FloatFromCurve(
     FloatCurve=DataInterface() {propertyValues {Curve {keys [<at 0 1>,<at 0.6 1>,<at 0.7 0>]}}}
     CurveIndex=$Particles.Visibility
     Scale Curve=0.25
    )
    ScaleA=true
    ScaleRGB=false
   ) {module=ScaleColor002}
   &ScaleColor(
    Scale Alpha=&FloatFromCurve(
     FloatCurve=DataInterface() {propertyValues {Curve {keys [
         {arriveTangent=4.3255, interpMode=RCIM_Cubic, leaveTangent=4.3255, tangentMode=RCTM_User, time=0.5, value=0}
         {arriveTangent=5.6231, interpMode=RCIM_Cubic, leaveTangent=5.6231, tangentMode=RCTM_User, time=1, value=1}
        ]}}}
     CurveIndex=$Particles.CameraScale
    )
    Scale RGB=&Multiply_VectorByFloat(
     Vector=&MakeVectorFromLinearColorRGB(Color=<rgba 1 0.0975 0.05 1>)
     Float=&Lerp_Float(
      A=0.5
      B=5
      Alpha=&FloatFromCurve(
       FloatCurve=DataInterface() {propertyValues {Curve {keys [
           {arriveTangent=4.3255, interpMode=RCIM_Cubic, leaveTangent=4.3255, tangentMode=RCTM_User, time=0.5, value=0}
           {arriveTangent=5.6231, interpMode=RCIM_Cubic, leaveTangent=5.6231, tangentMode=RCTM_User, time=1, value=1}
          ]}}}
       CurveIndex=$Particles.CameraScale
      )
     )
    )
    ScaleA=true
    ScaleRGB=true
   )
   &ScaleColor(
    Scale Alpha=&FN_Select_Float_from_Time_Of_Day(Morning=0, Day=0, Evening=0, Night=1)
    ScaleA=true
    ScaleRGB=false
   ) {module=ScaleColor001}
   &ScaleColor(
    Scale Alpha=&Multiply_Float(
     A=&MakeCustomFloatFromBool(Bool=$User.VehicleHasDriver, True Float=1, False Float=0)
     B=&MakeCustomFloatFromBool(Bool=$User.IsDrifting, True Float=1, False Float=0)
    )
    ScaleA=true
    ScaleRGB=false
   ) {module=ScaleColor003}
  ]
  rendererClasses [&NiagaraSpriteRendererProperties]
  renderers [
   {rendererClass=&NiagaraSpriteRendererProperties}
   {rendererClass=&NiagaraSpriteRendererProperties, rendererIndex=1}
  ]
 }
}
summary {
 systemName=NS_Wheel_And_Tire_Interaction_GetAway
 emitters [
  {emitterName=Sparks_Tires_Blown_Local, rendererClasses [&NiagaraSpriteRendererProperties]}
  {emitterName=Sparks_Tires_Blown_Local_Mesh, rendererClasses [&NiagaraMeshRendererProperties]}
  {emitterName=Dust_Tires_World, rendererClasses [&NiagaraSpriteRendererProperties]}
  {emitterName=Peel_Tires_Spikes, rendererClasses [&NiagaraSpriteRendererProperties]}
  {emitterName=Dust_Tires_Mesh, rendererClasses [&NiagaraMeshRendererProperties]}
  {emitterName=Dust_Debris, simTarget=GPUComputeSim, rendererClasses [&NiagaraMeshRendererProperties]}
  {emitterName=Exhaust_Local, rendererClasses [&NiagaraSpriteRendererProperties]}
  {emitterName=Sparks_Tires_Blown_World, rendererClasses [&NiagaraSpriteRendererProperties]}
  {emitterName=SnowChunkGPU, simTarget=GPUComputeSim, rendererClasses [&NiagaraSpriteRendererProperties]}
  {bEnabled=false, emitterName=NeonAnimTrail, rendererClasses [&NiagaraRibbonRendererProperties]}
  {emitterName=NeonLensFlare, simTarget=GPUComputeSim, rendererClasses [&NiagaraSpriteRendererProperties]}
  {emitterName=Peel_Tires_Smoke, rendererClasses [&NiagaraSpriteRendererProperties]}
  {emitterName=ExhaustFlame, rendererClasses [&NiagaraMeshRendererProperties, &NiagaraLightRendererProperties]}
  {emitterName=Jet_BurstSparksLow, rendererClasses [&NiagaraSpriteRendererProperties]}
  {emitterName=Exhaust_Local001, rendererClasses [&NiagaraSpriteRendererProperties]}
  {emitterName=AudioBlips, rendererClasses []}
 ]
 userVariables [
  User.BoostFX {type=NiagaraBool}
  User.Dust_Particles_Per_Wheel {type=NiagaraInt32, defaultValue=18}
  User.Dust_SpawnCurve {
   type=NiagaraDataInterfaceCurve
   description=""
   defaultValue {
    struct=NiagaraExt_VariableValue_DataInterface
    value {
     dataInterface=&self:NiagaraDataInterfaceCurve_0
     dataInterfaceClass=&NiagaraDataInterfaceCurve
    }
   }
  }
  User.Dust_Wheel_RL {type=NiagaraBool}
  User.Dust_Wheel_RR {type=NiagaraBool}
  User.ExhaustColor {type=LinearColor, description="", defaultValue {struct=LinearColor, value=<rgba 0.6062 0.754 1 0.05>}}
  User.ExhaustScale {type=NiagaraFloat, defaultValue=1, description=""}
  User.IsBigRig {type=NiagaraBool}
  User.IsBraking {type=NiagaraBool}
  User.IsDrifting {type=NiagaraBool}
  User.IsShifting {type=NiagaraBool}
  User.IsWaterDeep {type=NiagaraBool}
  User.Mod_OffroadTires {type=NiagaraBool}
  User.PeelOut_Wheel_RL {type=NiagaraBool, defaultValue=true}
  User.PeelOut_Wheel_RR {type=NiagaraBool, defaultValue=true}
  User.PeelOutColorArray {
   type=NiagaraDataInterfaceArrayColor
   description=""
   defaultValue {
    struct=NiagaraExt_VariableValue_DataInterface
    value {
     dataInterface=&self:NiagaraDataInterfaceArrayColor_0
     dataInterfaceClass=&NiagaraDataInterfaceArrayColor
    }
   }
  }
  User.PeelOutColorIndex {type=NiagaraInt32}
  User.RibbonIDOffset {type=NiagaraInt32}
  User.SedanColor {type=NiagaraInt32}
  User.SK_Vehicle {
   type=NiagaraDataInterfaceSkeletalMesh
   defaultValue {
    struct=NiagaraExt_VariableValue_DataInterface
    value {dataInterface=&self:NiagaraDataInterfaceSkeletalMesh_1}
   }
  }
  User.SkeletalMesh {
   type=NiagaraDataInterfaceSkeletalMesh
   defaultValue {
    struct=NiagaraExt_VariableValue_DataInterface
    value {dataInterface=&self:NiagaraDataInterfaceSkeletalMesh_2}
   }
  }
  User.Slide_Direction {type=NiagaraFloat, defaultValue=0, description=""}
  User.Sparks_Wheel_ML {type=NiagaraBool}
  User.Sparks_Wheel_MR {type=NiagaraBool}
  User.SparkSPU_MaxSpeed {type=NiagaraFloat, defaultValue=3, description=""}
  User.SparkSPU_MinSpeed {type=NiagaraFloat, defaultValue=4, description=""}
  User.Tire_Scale_Back {type=Vector3f, description="", defaultValue {struct=Vector3f, value=<0.39 0.4325 0.39>}}
  User.TireIntactFL {type=NiagaraBool, defaultValue=true}
  User.TireIntactFR {type=NiagaraBool, defaultValue=true}
  User.TireIntactRL {type=NiagaraBool, defaultValue=true}
  User.TireIntactRR {type=NiagaraBool, defaultValue=true}
  User.Vehicle_Speed {type=NiagaraFloat, defaultValue=0, description=""}
  User.VehicleHasDriver {type=NiagaraBool}
  User.VehicleHasFuel {type=NiagaraBool}
  User.Wheel_Scale_Back {type=Vector3f, description="", defaultValue {struct=Vector3f, value=<0.3 0.3 0.3>}}
  User.Wheel_Scale_Front {type=Vector3f, description="", defaultValue {struct=Vector3f, value=<0.3 0.3 0.3>}}
  User.Wheel_Width {type=NiagaraFloat, defaultValue=16, description=""}
  User.Wheel_Width_Back {type=NiagaraFloat, defaultValue=18, description=""}
  User.WheelsTouchingArray {
   type=NiagaraDataInterfaceArrayFloat
   defaultValue {
    struct=NiagaraExt_VariableValue_DataInterface
    value {
     dataInterface=&self:NiagaraDataInterfaceArrayFloat_0
     dataInterfaceClass=&NiagaraDataInterfaceArrayFloat
    }
   }
  }
 ]
}
